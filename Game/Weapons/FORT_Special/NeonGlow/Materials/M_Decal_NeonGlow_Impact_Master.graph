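# Translucent unlit deferred decal master material for NeonGlow weapon impacts, drawing a glowing scorch mark that fades out.
# A parameter-selected palette color is multiplied by a scorch mask broken up with rotating plasma noise.
# Age from Spawntime and Lifetime, or a looping debug switch, dims the color, sharpens the mask and fades emissive to black.
# A ground-char texture with its own timed fade is inverted and subtracted to erode the glow.
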Material /Game/Weapons/FORT_Special/NeonGlow/Materials/M_Decal_NeonGlow_Impact_Master
graph {
 properties {
  bAutomaticallySetUsageInEditor=false
  blendMode=BLEND_Translucent
  materialDomain=MD_DeferredDecal
  pixelDepthOffsetMode=PDOM_Legacy
  shadingModel=MSM_Unlit
  usedShadingModels=Unlit
 }
 nodes [
  n1 = TextureSampleParameter2D(UVs=AppendVector(A=Constant() {r=1}, B=ScalarParameter() {defaultValue=0.3, parameterName=Pulse_Color_Selection})) {group=Neon, parameterName=Neon_Emissive, texture=/Game/Weapons/FORT_Rifles/Textures/Neon_Glow_GlowParts_E}
  n2 = Multiply(A=@n1.RGB, B=ScalarParameter() {defaultValue=20, group=Neon, parameterName=Emis Intensity})
  n3 = Time()
  n4 = Time()
  n5 = ScalarParameter() {group=Timestamp, parameterName=Spawntime}
  n6 = Divide(
   A=StaticSwitchParameter(True=Frac(None=@n3), False=Subtract(A=@n4, B=@n5)) {group=Timestamp, parameterName=Debug Timestamp}
   B=ScalarParameter() {defaultValue=5, group=Timestamp, parameterName=Lifetime}
  )
  n7 = Clamp(None=Multiply(A=@n6, B=ScalarParameter() {defaultValue=2, group=Emissive, parameterName=Emissive Transition Multiplier}) {constB=6})
  n8 = TextureSampleParameter2D() {parameterName=DecalTexture_1, texture=/Game/Effects/Fort_Effects/Gameplay/SmokeDustExplosions/Textures/T_Decal_Scorch_Hole}
  n9 = Rotator(Coordinate=TextureCoordinate() {vTiling=0.5})
  n10 = Add(A=@n8.R, B=Multiply(A=@n8.R, B=TextureSample(UVs=@n9) {texture=/Game/Effects/Fort_Effects/Textures/Plasma/T_PlasmaNoise_A_01}.R))
  n11 = LinearInterpolate(Alpha=@n7) {constB=25}
  n12 = TextureSampleParameter2D() {
   parameterName=DecalTexture
   samplerType=SAMPLERTYPE_LinearColor
   texture=/Game/Effects/Fort_Effects/Gameplay/SmokeDustExplosions/Textures/T_GroundChar_01
  }
  n13 = OneMinus(None=@n6)
  n14 = ScalarParameter() {defaultValue=4, group=Timestamp, parameterName=Fade Transition Multiplier}
  n15 = LinearInterpolate(
   B=LinearInterpolate(B=@n12.RGB, Alpha=ScalarParameter() {defaultValue=0.9, parameterName=Opacity}) {constB=0}
   Alpha=Clamp(None=Multiply(A=@n13, B=@n14) {constB=4})
  ) {constB=0}
  n16 = Subtract(
   A=Multiply(A=LinearInterpolate(A=@n2, B=@n1.RGB, Alpha=@n7), B=Power(Base=@n10, Exp=@n11) {constExponent=0})
   B=Multiply(A=OneMinus(None=@n15), B=ScalarParameter() {defaultValue=0.55, parameterName=Erode Color Scale})
  )
  MaterialFunctionCall(Emissive=LinearInterpolate(A=Clamp(None=@n16), Alpha=@n7) {constB=0}) {materialFunction=&MF_FX_EmissiveAdjustments}
  ScalarParameter() {parameterName=Metallic}
  ScalarParameter() {parameterName=Specular}
  ScalarParameter() {defaultValue=1, parameterName=Roughness}
  VectorParameter() {parameterName=DecalColor, defaultValue=<rgba 0 0 0 1>}
 ]
}
