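# Unlit, two-sided translucent mesh-particle material for the NeonGlow weapon's outward neon gas shot effect.
# Stepped panning smoke textures form two phases, blended in steps by particle blue, while red and green scale pan time.
# A dynamic parameter selects glow colors from a texture; another drives UV-masked offset along normals.
# Thresholded noise plus particle alpha erodes opacity, with depth fade on the non-mobile path.
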Material /Game/Weapons/FORT_Special/NeonGlow/Materials/M_NeonGas_Shoot_Out
graph {
 properties {
  bUsedWithMeshParticles=true
  blendMode=BLEND_Translucent
  pixelDepthOffsetMode=PDOM_Legacy
  shadingModel=MSM_Unlit
  twoSided=true
  usedShadingModels=Unlit
 }
 nodes [
  Constant()
  n1 = TextureSampleParameter2D(UVs=AppendVector(
    A=Constant() {r=1}
    B=DynamicParameter() {defaultValue=<rgba 1 0.3 1 1>, paramNames [WPO_Lerp_Control, Pulse_Color_Selection, Param3, Param4]}.Pulse_Color_Selection
   )) {group=Neon, parameterName=Neon_Emissive, texture=/Game/Weapons/FORT_Rifles/Textures/Neon_Glow_GlowParts_E}
  n2 = Multiply(A=@n1.RGB, B=ScalarParameter() {defaultValue=10, parameterName=Outline_Phase1_Intensity})
  n3 = Multiply(A=@n1.RGB, B=ScalarParameter() {defaultValue=20, parameterName=Core_Phase1_Intensity})
  n4 = AppendVector(
   A=ScalarParameter() {defaultValue=2, group=Phase 1, parameterName=Muzzle A1 U Repeat}
   B=ScalarParameter() {defaultValue=1, group=Phase 1, parameterName=Muzzle A1 U Repeat_1}
  )
  n5 = TextureCoordinate()
  n6 = Multiply(A=ParticleColor().R, B=Multiply(A=Time(), B=ScalarParameter() {defaultValue=8.5, group=Phase 1, parameterName=Pan Speed Master A1}))
  n7 = TextureSample(UVs=Panner(Coordinate=Multiply(A=@n4, B=@n5), Time=Ceil(None=@n6)) {bFractionalPart=true, speedX=0.2}) {samplerType=SAMPLERTYPE_Grayscale, texture=&T_Smoke_Trail_Soft}
  n8 = TextureSampleParameter2D(UVs=AppendVector(
    A=Constant() {r=1}
    B=DynamicParameter() {defaultValue=<rgba 1 0.3 1 1>, paramNames [WPO_Lerp_Control, Pulse_Color_Selection, Param3, Param4]}.Pulse_Color_Selection
   )) {group=Neon, parameterName=Neon_Emissive, texture=&Neon_Glow_GlowParts_E}
  n9 = Multiply(A=@n8.RGB, B=ScalarParameter() {defaultValue=50, parameterName=Outline_Phase2_Intensity})
  n10 = Multiply(A=@n8.RGB, B=ScalarParameter() {parameterName=Core_Phase2_Intensity})
  n11 = TextureCoordinate()
  n12 = AppendVector(
   A=ScalarParameter() {defaultValue=1.35, group=Phase 2, parameterName=Muzzle B2 U Repeat}
   B=ScalarParameter() {defaultValue=0.18, group=Phase 2, parameterName=Muzzle B2 U Repeat_1}
  )
  n13 = Multiply(A=ParticleColor().G, B=Multiply(A=Time(), B=ScalarParameter() {defaultValue=13.1, group=Phase 2, parameterName=Pan Speed Master B2}))
  n14 = TextureSample(UVs=Panner(Coordinate=Multiply(A=@n11, B=@n12), Time=Ceil(None=@n13)) {bFractionalPart=true, speedX=2.2, speedY=0.4}) {samplerType=SAMPLERTYPE_Grayscale, texture=&T_Smoke_Trail_Soft}
  n15 = Time()
  n16 = ScalarParameter() {defaultValue=3, group=LERP, parameterName=Steps}
  n17 = Multiply(A=Clamp(None=ParticleColor().B), B=@n16)
  n18 = StaticSwitchParameter(True=ConstantBiasScale(None=Sine(None=@n15)), False=Divide(A=Ceil(None=@n17), B=@n16)) {parameterName=Debug LERP?}
  MaterialFunctionCall(Emissive=LinearInterpolate(A=LinearInterpolate(A=@n2, B=@n3, Alpha=@n7.R), B=LinearInterpolate(A=@n9, B=@n10, Alpha=@n14.R), Alpha=@n18)) {materialFunction=&MF_FX_EmissiveAdjustments}
  n19 = TextureCoordinate()
  n20 = ScalarParameter() {defaultValue=-50, parameterName=WPO_Amount_Min}
  n21 = ScalarParameter() {defaultValue=50, parameterName=WPO_Amount_Max}
  n22 = DynamicParameter() {defaultValue=<rgba 1 0.3 1 1>, paramNames [WPO_Lerp_Control, Pulse_Color_Selection, Param3, Param4]}
  Multiply(
   A=OneMinus(None=ComponentMask(None=@n19) {g=true})
   B=Multiply(A=VertexNormalWS(), B=LinearInterpolate(A=@n20, B=@n21, Alpha=@n22.WPO_Lerp_Control))
  )
  n23 = TextureCoordinate()
  n24 = Constant2Vector() {g=-0.2}
  n25 = ScalarParameter() {defaultValue=1.5, group=Phase 1, parameterName=Muzzle B1 U Repeat}
  n26 = ScalarParameter() {defaultValue=1, group=Phase 1, parameterName=Muzzle B1 U Repeat_1}
  n27 = ParticleColor()
  n28 = Multiply(A=Time(), B=ScalarParameter() {defaultValue=13.1, group=Phase 1, parameterName=Pan Speed Master B1})
  n29 = Panner(Coordinate=Multiply(A=Add(A=@n23, B=@n24), B=AppendVector(A=@n25, B=@n26)), Time=Ceil(None=Multiply(A=@n27.G, B=@n28))) {bFractionalPart=true, speedX=-0.12}
  n30 = Clamp(None=Add(
    A=@n7.R
    B=TextureSample(UVs=@n29) {samplerType=SAMPLERTYPE_Grayscale, texture=&T_Smoke_Trail_Soft}.G
   ))
  n31 = Multiply(
   A=TextureCoordinate()
   B=AppendVector(
    A=ScalarParameter() {defaultValue=1.1, group=Phase 2, parameterName=Muzzle A2 U Repeat}
    B=ScalarParameter() {defaultValue=0.1, group=Phase 2, parameterName=Muzzle A2 U Repeat_1}
   )
  )
  n32 = Time()
  n33 = ScalarParameter() {defaultValue=8.5, group=Phase 2, parameterName=Pan Speed Master A2}
  n34 = Ceil(None=Multiply(A=ParticleColor().R, B=Multiply(A=@n32, B=@n33)))
  n35 = Add(
   A=TextureSample(UVs=Panner(Coordinate=@n31, Time=@n34) {bFractionalPart=true, speedX=4.8, speedY=-0.3}) {samplerType=SAMPLERTYPE_Grayscale, texture=&T_Smoke_Trail_Soft}.R
   B=@n14.R
  )
  n36 = Power(
   Base=LinearInterpolate(A=@n30, B=Multiply(A=@n30, B=@n35), Alpha=@n18)
   Exp=ScalarParameter() {defaultValue=4, group=Erode, parameterName=Lerp Exponent}
  ) {constExponent=10}
  n37 = ScalarParameter() {defaultValue=50, group=Erode, parameterName=Lerp Brightness}
  n38 = ScalarParameter() {defaultValue=1.8, group=Erode, parameterName=Erode U Repeat A}
  n39 = ScalarParameter() {defaultValue=0.3, group=Erode, parameterName=Erode V Repeat A}
  n40 = ParticleColor()
  n41 = Multiply(A=Time(), B=ScalarParameter() {defaultValue=8.5, group=Phase 2, parameterName=Pan Speed Master A3})
  n42 = Panner(Coordinate=Multiply(A=AppendVector(A=@n38, B=@n39), B=TextureCoordinate()), Time=Ceil(None=Multiply(A=@n40.R, B=@n41))) {bFractionalPart=true, speedX=-0.2, speedY=2.4}
  n43 = ScalarParameter() {defaultValue=1.3, group=Erode, parameterName=Erode U Repeat B}
  n44 = ScalarParameter() {defaultValue=0.55, group=Erode, parameterName=Erode V Repeat B}
  n45 = ParticleColor()
  n46 = Multiply(A=Time(), B=ScalarParameter() {defaultValue=13.1, group=Phase 2, parameterName=Pan Speed Master B3})
  n47 = Panner(Coordinate=Multiply(A=TextureCoordinate(), B=AppendVector(A=@n43, B=@n44)), Time=Ceil(None=Multiply(A=@n45.G, B=@n46))) {bFractionalPart=true, speedX=0.2, speedY=1.15}
  n48 = Add(
   A=Add(
    A=TextureSample(UVs=@n42) {texture=&T_NoisePack_01}.R
    B=TextureSample(UVs=@n47) {texture=&T_NoisePack_01}.B
   )
   B=ParticleColor().A
  )
  n49 = Multiply(A=Clamp(None=Multiply(A=@n36, B=@n37)), B=Floor(None=Clamp(None=@n48)))
  ShadingPathSwitch(Default=DepthFade(Opacity=@n49, FadeDistance=ScalarParameter() {defaultValue=50, parameterName=Fade_Distance}), Mobile=@n49) {desc=Migrated from FeatureLevelSwitch}
 ]
}
