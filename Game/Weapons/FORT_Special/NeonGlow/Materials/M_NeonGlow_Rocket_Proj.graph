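# Additive, unlit, two-sided mesh-particle material for the NeonGlow rocket projectile, drawing animated energy arcs in neon colors.
# Two counter-panning, mirrored energy-flow samples use time-driven random channel picking, combining with a sine band into a mask.
# That mask blends two colors picked at fixed UVs from the weapon's glow texture, with optional desaturation and LDR-to-HDR remapping.
# Particle color, depth fade (non-mobile), a smooth-tile fade and a sine arc sweep shape the output.
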Material /Game/Weapons/FORT_Special/NeonGlow/Materials/M_NeonGlow_Rocket_Proj
graph {
 properties {
  bUsedWithMeshParticles=true
  blendMode=BLEND_Additive
  pixelDepthOffsetMode=PDOM_Legacy
  shadingModel=MSM_Unlit
  twoSided=true
  usedShadingModels=Unlit
 }
 nodes [
  n1 = ParticleColor()
  n2 = AppendVector(A=Constant() {r=1}, B=ScalarParameter() {defaultValue=0.9, group=Neon, parameterName=Pulse_Color_Selection_1, sortPriority=2})
  n3 = ScalarParameter() {defaultValue=5, parameterName=EmissiveIntensity_1}
  n4 = DynamicParameter() {defaultValue=<rgba 1 0 1 1>, paramNames [Overall_Intensity, X_Offset, Param3, Param4]}
  n5 = AppendVector(A=Constant() {r=1}, B=ScalarParameter() {defaultValue=0.775, group=Neon, parameterName=Pulse_Color_Selection, sortPriority=2})
  n6 = ScalarParameter() {defaultValue=5, parameterName=EmissiveIntensity}
  n7 = DynamicParameter() {defaultValue=<rgba 1 0 1 1>, paramNames [Overall_Intensity, X_Offset, Param3, Param4]}
  n8 = Multiply(A=Time() {period=1}, B=MaterialFunctionCall() {materialFunction=&DebugTimeSine}.Result)
  n9 = ScalarParameter() {defaultValue=1, parameterName=XTileMultiplier}
  n10 = ScalarParameter() {defaultValue=1, parameterName=YTileMultiplier}
  n11 = MaterialFunctionCall(Float2=Multiply(A=TextureCoordinate(), B=AppendVector(A=@n9, B=@n10))) {materialFunction=&BreakOutFloat2Components}
  n12 = AppendVector(
   A=Add(A=DynamicParameter() {defaultValue=<rgba 1 0 1 1>, paramNames [Overall_Intensity, X_Offset, Param3, Param4]}.X_Offset, B=@n11.R)
   B=@n11.G
  )
  n13 = MaterialFunctionCall!(
   Random Picker=@n8
   Texture Input=TextureSampleParameter2D(UVs=Panner(Coordinate=@n12) {speedX=1}) {parameterName=Texture1, texture=/Game/Effects/Fort_Effects/Textures/Energy/T_EnergyFlow_Packed}.RGB
   Option 1
   Option 2
   Option 3
  ) {materialFunction=/Game/Effects/Fort_Effects/Materials/_Functions/MF_RGB_Random_01}
  n14 = MaterialFunctionCall(Float2=@n12) {materialFunction=&BreakOutFloat2Components}
  n15 = AppendVector(A=@n14.R, B=OneMinus(None=@n14.G))
  n16 = MaterialFunctionCall(
   Random Picker=@n8
   Texture Input=TextureSampleParameter2D(UVs=Panner(Coordinate=@n15) {speedX=-1}) {parameterName=Texture2, texture=&T_EnergyFlow_Packed}.RGB
  ) {materialFunction=&MF_RGB_Random_01}
  n17 = TextureCoordinate()
  n18 = Power(Base=Sine(None=ComponentMask(None=@n17) {g=true}) {period=2}, Exp=Constant() {r=6})
  n19 = Clamp(None=Add(A=Add(A=@n13.Result, B=@n16.Result), B=Multiply(B=@n18) {constA=0.25}))
  n20 = LinearInterpolate(
   A=Multiply(
    A=TextureSampleParameter2D(UVs=@n2) {group=Neon, parameterName=Neon_Emissive_1, texture=/Game/Weapons/FORT_Rifles/Textures/Neon_Glow_GlowParts_E}.RGB
    B=Multiply(A=@n3, B=@n4.Overall_Intensity)
   ) {constB=3}
   B=Multiply(
    A=TextureSampleParameter2D(UVs=@n5) {group=Neon, parameterName=Neon_Emissive, texture=&Neon_Glow_GlowParts_E}.RGB
    B=Multiply(A=@n6, B=@n7.Overall_Intensity)
   ) {constB=3}
   Alpha=@n19
  )
  n21 = ShadingPathSwitch(Default=DepthFade(FadeDistance=ScalarParameter() {defaultValue=50, parameterName=Depth Fade Distance}), Mobile=Constant() {r=1}) {desc=Migrated from FeatureLevelSwitch}
  n22 = Multiply(
   A=LinearInterpolate(A=@n20, B=Desaturation(None=@n20), Alpha=ScalarParameter() {parameterName=Desaturate?, sliderMax=1})
   B=Multiply(A=@n19, B=@n21)
  )
  n23 = Clamp(None=@n22)
  n24 = TextureSample() {texture=&T_Smooth_tile}
  n25 = LinearInterpolate(Alpha=ScalarParameter() {parameterName=Fade_Control, sliderMax=1}) {constA=-1, constB=5}
  n26 = ComponentMask(None=TextureCoordinate()) {r=true}
  n27 = ScalarParameter() {parameterName=Arc_Sweep, sliderMax=1}
  n28 = Multiply(
   A=Multiply(A=Multiply(A=@n19, B=@n21), B=Power(Base=@n24.R, Exp=@n25))
   B=Power(Base=Sine(None=@n26) {period=2}, Exp=LinearInterpolate(Alpha=@n27) {constA=5, constB=100})
  )
  n29 = LinearInterpolate(Alpha=Clamp(None=@n28)) {constA=0.5, constB=2}
  MaterialFunctionCall(Emissive=Multiply(
    A=@n1.RGB
    B=MaterialFunctionCall(LDR Input=@n23, HDR Tint=@n22, HDR Intensity Multiplier=@n29) {materialFunction=&DeriveHDRfromLDR}.RemappedRange
   )) {materialFunction=&MF_FX_EmissiveAdjustments}
  Multiply(A=@n1.A, B=@n28)
 ]
}
