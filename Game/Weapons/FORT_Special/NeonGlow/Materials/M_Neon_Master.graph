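# Master material for NeonGlow weapon parts, combining glossy glass surface parameters with neon emissive and a normal map.
# Emissive sums the glow texture, a tiled ramp boost and a pulse along U, scrolled by time or a manual-control switch.
# Pulse color is sampled from the glow texture at a selectable UV, and the sum is remapped to HDR.
# Distance-cull LOD and black hole functions build a vertex offset.
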Material /Game/Weapons/FORT_Special/NeonGlow/Materials/M_Neon_Master
graph {
 properties {
  bScreenSpaceReflections=true
  bUsedWithMeshDeformer=true
  bUsedWithSkeletalMesh=true
  physMaterial=/Game/FortressPhysicalMaterials/Glass
  pixelDepthOffsetMode=PDOM_Legacy
  translucencyDirectionalLightingIntensity=5
  translucencyLightingMode=TLM_VolumetricPerVertexDirectional
  usedShadingModels=Default Lit
 }
 nodes [
  n1 = VectorParameter() {parameterName=WorldPostionOfBlackHole, defaultValue=<rgba -30000 25536 1952 0>}
  n2 = MaterialFunctionCall!(
   World Position
   DistanceCullFadeAnimationOffset=MaterialFunctionCall() {materialFunction=/Game/Packages/Fortress_SharedMaterials/Base_Material_Functions/PickupEffectDistanceCullOverride}.Result
   Use actor center instead of pivot point
  ) {materialFunction=/Game/Packages/Fortress_SharedMaterials/Base_Material_Functions/LODEffect}
  n3 = WorldPosition()
  Add(
   A=MaterialFunctionCall!(WorldPostionOfBlackHole=AppendVector(A=@n1.RGB, B=@n1.RGB), World Position=Add(A=@n2.Result, B=@n3.XYZ)) {materialFunction=/Game/Packages/Fortnite_UniqueMaterials_1/BlackHoleFunction}.Result
   B=@n2.Result
  )
  n4 = TextureSampleParameter2D() {group=Neon, parameterName=Neon_Emissive, texture=/Game/Weapons/FORT_Rifles/Textures/Neon_Glow_GlowParts_E}
  n5 = TextureCoordinate() {vTiling=4}
  n6 = Multiply(
   A=Multiply(A=@n4.RGB, B=TextureSample(UVs=@n5) {texture=/Game/Global/Textures/Default/Masks_and_Gradients/T_Water_Top_Ramp}.RGB)
   B=ScalarParameter() {defaultValue=2, group=Neon, parameterName=Central_Intensity}
  )
  n7 = ScalarParameter() {group=Neon, parameterName=Pulse_Control}
  n8 = Frac(None=Multiply(A=Time() {period=1}, B=ScalarParameter() {defaultValue=0.5, group=Neon, parameterName=Pulse_Speed}))
  n9 = Add(
   A=ComponentMask(None=TextureCoordinate()) {r=true}
   B=LinearInterpolate(Alpha=StaticSwitchParameter(True=@n7, False=@n8) {parameterName=Manual_Control_Pulse}) {constA=1, constB=-1}
  ) {constB=0}
  n10 = Power(
   Base=Sine(None=Clamp(None=@n9)) {period=2}
   Exp=ScalarParameter() {defaultValue=10, group=Neon, parameterName="Pulse_SizeConrol (Contrast)", sortPriority=3}
  ) {constExponent=6}
  n11 = TextureSampleParameter2D(UVs=AppendVector(A=Constant() {r=1}, B=ScalarParameter() {defaultValue=0.3, group=Neon, parameterName=Pulse_Color_Selection, sortPriority=2})) {group=Neon, parameterName=Neon_Emissive, texture=&Neon_Glow_GlowParts_E}
  n12 = Add(
   A=Multiply(A=Add(A=@n4.RGB, B=@n6), B=ScalarParameter() {defaultValue=1.5, group=Neon, parameterName=EmissiveMultiplier, sortPriority=1})
   B=Multiply(A=Multiply(A=@n10, B=@n11.RGB), B=ScalarParameter() {defaultValue=5, group=Neon, parameterName=Pulse_Intensity, sortPriority=4})
  )
  MaterialFunctionCall(Emissive=MaterialFunctionCall(LDR Input=Clamp(None=@n12), HDR Tint=@n12, HDR Intensity Multiplier=ScalarParameter() {defaultValue=0.25, parameterName=Neon HDR INtensity}) {materialFunction=&DeriveHDRfromLDR}.RemappedRange) {materialFunction=&MF_FX_EmissiveAdjustments}
  ScalarParameter() {group=Glass, parameterName=Metallic, sortPriority=1}
  ScalarParameter() {defaultValue=0.2286, group=Glass, parameterName=Roughness, sortPriority=3}
  ScalarParameter() {defaultValue=1, group=Glass, parameterName=Specular, sortPriority=2}
  TextureSample() {samplerType=SAMPLERTYPE_Normal, texture=/Game/Weapons/FORT_Rifles/Textures/Neon_Glow_GlowParts_N}
 ]
}
